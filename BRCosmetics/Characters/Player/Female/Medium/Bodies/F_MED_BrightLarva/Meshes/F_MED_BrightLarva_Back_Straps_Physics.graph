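# Physics asset for the back straps of the BrightLarva medium female player body cosmetic, simulating two hanging tail chains.
# A main tail body has tight swing limits against its base and joins small left and right base bodies through joints with free swing and twist.
# From each base runs a five-capsule chain with limited swing and zero twist limit, slightly tighter on the right.
# Two larger back-skirt capsules have no constraints.
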PhysicsAsset /BRCosmetics/Characters/Player/Female/Medium/Bodies/F_MED_BrightLarva/Meshes/F_MED_BrightLarva_Back_Straps_Physics
bodies {
 base_main_tail [{shapeType=Capsule, length=0.505, radius=0.505, shapeName=base_main_tail_capsule, rotation=<rot 0 -90 0>}]
 base_tail_l_0 [{shapeType=Capsule, length=0.505, radius=0.505, shapeName=base_tail_l_0_capsule, rotation=<rot 0 -90 0>}]
 base_tail_r_0 [{shapeType=Capsule, length=0.505, radius=0.505, shapeName=base_tail_r_0_capsule, rotation=<rot 0 -90 0>}]
 dyn_main_tail [{shapeType=Capsule, length=3, radius=2, shapeName=dyn_main_tail_capsule, center=<2 0 0>, rotation=<rot 90 0 0>}]
 dyn_skirt_back_1 [{shapeType=Capsule, length=16.365, radius=10.3447, shapeName=dyn_skirt_back_1_capsule, center=<0.5727 7.6156 0>}]
 dyn_skirt_back_2 [{shapeType=Capsule, length=16.0832, radius=10.6265, shapeName=dyn_skirt_back_1_capsule, center=<4.79 6.2033 0>}]
 dyn_tail_l_0 [{shapeType=Capsule, length=9, radius=1.5, shapeName=dyn_tail_l_0_capsule, center=<2 0 0>, rotation=<rot 90 0 0>}]
 dyn_tail_l_1 [{shapeType=Capsule, length=9, radius=1.5, shapeName=dyn_tail_l_1_capsule, center=<2 0 0>, rotation=<rot 90 0 0>}]
 dyn_tail_l_2 [{shapeType=Capsule, length=9, radius=1.5, shapeName=dyn_tail_l_2_capsule, center=<2 0 0>, rotation=<rot 90 0 0>}]
 dyn_tail_l_3 [{shapeType=Capsule, length=9, radius=1.5, shapeName=dyn_tail_l_3_capsule, center=<2 0 0>, rotation=<rot 90 0 0>}]
 dyn_tail_l_4 [{shapeType=Capsule, length=9, radius=1.5, shapeName=dyn_tail_l_4_capsule, center=<2 0 0>, rotation=<rot 90 0 0>}]
 dyn_tail_r_0 [{shapeType=Capsule, length=9, radius=1.5, shapeName=dyn_tail_r_0_capsule, center=<2 0 0>, rotation=<rot 90 0 0>}]
 dyn_tail_r_1 [{shapeType=Capsule, length=9, radius=1.5, shapeName=dyn_tail_r_1_capsule, center=<2 0 0>, rotation=<rot 90 0 0>}]
 dyn_tail_r_2 [{shapeType=Capsule, length=9, radius=1.5, shapeName=dyn_tail_r_2_capsule, center=<2 0 0>, rotation=<rot 90 0 0>}]
 dyn_tail_r_3 [{shapeType=Capsule, length=9, radius=1.5, shapeName=dyn_tail_r_3_capsule, center=<2 0 0>, rotation=<rot 90 0 0>}]
 dyn_tail_r_4 [{shapeType=Capsule, length=9, radius=1.5, shapeName=dyn_tail_r_4_capsule, center=<2 0 0>, rotation=<rot 90 0 0>}]
}
constraints [
 {bone1Name=dyn_tail_l_0, bone2Name=base_tail_l_0, swing1LimitDegrees=20, swing2LimitDegrees=30, twistLimitDegrees=0}
 {bone1Name=dyn_tail_l_1, bone2Name=dyn_tail_l_0, swing1LimitDegrees=20, swing2LimitDegrees=30, twistLimitDegrees=0}
 {bone1Name=dyn_tail_l_2, bone2Name=dyn_tail_l_1, swing1LimitDegrees=20, swing2LimitDegrees=30, twistLimitDegrees=0}
 {bone1Name=dyn_tail_l_3, bone2Name=dyn_tail_l_2, swing1LimitDegrees=20, swing2LimitDegrees=30, twistLimitDegrees=0}
 {bone1Name=dyn_tail_l_4, bone2Name=dyn_tail_l_3, swing1LimitDegrees=20, swing2LimitDegrees=30, twistLimitDegrees=0}
 {bone1Name=dyn_tail_r_0, bone2Name=base_tail_r_0, swing1LimitDegrees=15, swing2LimitDegrees=30, twistLimitDegrees=0}
 {bone1Name=dyn_tail_r_1, bone2Name=dyn_tail_r_0, swing1LimitDegrees=15, swing2LimitDegrees=30, twistLimitDegrees=0}
 {bone1Name=dyn_tail_r_2, bone2Name=dyn_tail_r_1, swing1LimitDegrees=15, swing2LimitDegrees=30, twistLimitDegrees=0}
 {bone1Name=dyn_tail_r_3, bone2Name=dyn_tail_r_2, swing1LimitDegrees=15, swing2LimitDegrees=30, twistLimitDegrees=0}
 {bone1Name=dyn_tail_r_4, bone2Name=dyn_tail_r_3, swing1LimitDegrees=15, swing2LimitDegrees=30, twistLimitDegrees=0}
 {
  bone1Name=base_tail_r_0
  bone2Name=dyn_main_tail
  swing1LimitDegrees=45
  swing1Motion=Free
  swing2LimitDegrees=45
  swing2Motion=Free
  twistLimitDegrees=45
  twistMotion=Free
 }
 {
  bone1Name=base_tail_l_0
  bone2Name=dyn_main_tail
  swing1LimitDegrees=45
  swing1Motion=Free
  swing2LimitDegrees=45
  swing2Motion=Free
  twistLimitDegrees=45
  twistMotion=Free
 }
 {bone1Name=dyn_main_tail, bone2Name=base_main_tail, swing1LimitDegrees=10, swing2LimitDegrees=10, twistLimitDegrees=0}
]
