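# Physics asset for the chest bows on the F_MED_BrightLarva female player body cosmetic, defining capsule bodies for their secondary motion.
# Four bow chains (large left and right, "sh" left, plain right) each link three dyn bones in series to a small base body.
# Constraints allow only narrow swing with zero twist, and the large bows loosen past the first link.
# An unconstrained spine_05 capsule provides a torso volume.
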PhysicsAsset /BRCosmetics/Characters/Player/Female/Medium/Bodies/F_MED_BrightLarva/Meshes/F_MED_BrightLarva_Chest_Bows_Physics
bodies {
 base_bow_chest_r_0 [{shapeType=Capsule, length=0.505, radius=0.505, shapeName=base_bow_chest_r_0_capsule, rotation=<rot 0 -90 0>}]
 base_lrg_bow_chest_l_0 [{shapeType=Capsule, length=0.505, radius=0.505, shapeName=base_lrg_bow_chest_l_0_capsule, rotation=<rot 0 -90 0>}]
 base_lrg_bow_chest_r_0 [{shapeType=Capsule, length=0.505, radius=0.505, shapeName=base_lrg_bow_chest_r_0_capsule, rotation=<rot 0 -90 0>}]
 base_sh_bow_chest_l_0 [{shapeType=Capsule, length=0.505, radius=0.505, shapeName=base_sh_bow_chest_l_0_capsule, rotation=<rot 0 -90 0>}]
 dyn_bow_chest_r_0 [{shapeType=Capsule, length=3.4993, radius=1.3007, shapeName=dyn_bow_chest_r_0_capsule, center=<2 -0.5 0>}]
 dyn_bow_chest_r_1 [{shapeType=Capsule, length=3.2298, radius=1.3007, shapeName=dyn_bow_chest_r_1_capsule, center=<2 -0.5 0>}]
 dyn_bow_chest_r_2 [{shapeType=Capsule, length=3.4823, radius=1.3177, shapeName=dyn_bow_chest_r_2_capsule, center=<2 0 0>, rotation=<rot 0 -0.0001 -1.1442>}]
 dyn_lrg_bow_chest_l_0 [{
   shapeType=Capsule
   length=4.0064
   radius=0.7936
   shapeName=dyn_lrg_bow_chest_l_0_capsule
   center=<2 -1 0>
   rotation=<rot 84.9119 180 -179.9999>
  }]
 dyn_lrg_bow_chest_l_1 [{
   shapeType=Capsule
   length=3.3878
   radius=1.1027
   shapeName=dyn_lrg_bow_chest_l_1_capsule
   center=<2.018 -1.5 0>
   rotation=<rot 86.091 -179.6417 172.1714>
  }]
 dyn_lrg_bow_chest_l_2 [{
   shapeType=Capsule
   length=2.7237
   radius=1.4142
   shapeName=dyn_lrg_bow_chest_l_2_capsule
   center=<2 -1.0856 0>
   rotation=<rot 3.0739 -0.0001 8.2062>
  }]
 dyn_lrg_bow_chest_r_0 [{shapeType=Capsule, length=4.0064, radius=0.7936, shapeName=dyn_lrg_bow_chest_r_0_capsule, center=<2 -1 0>, rotation=<rot 86.2334 0 0.0001>}]
 dyn_lrg_bow_chest_r_1 [{
   shapeType=Capsule
   length=3.159
   radius=1.1027
   shapeName=dyn_lrg_bow_chest_r_1_capsule
   center=<2.0043 -1.2709 0>
   rotation=<rot 81.6103 0.0967 -4.0929>
  }]
 dyn_lrg_bow_chest_r_2 [{
   shapeType=Capsule
   length=2.6544
   radius=1.4246
   shapeName=dyn_lrg_bow_chest_r_2_capsule
   center=<2 -1.288 0>
   rotation=<rot 10.0669 179.995 -169.7872>
  }]
 dyn_sh_bow_chest_l_0 [{shapeType=Capsule, length=3.4993, radius=1.3007, shapeName=dyn_sh_bow_chest_l_0_capsule, center=<2 -0.5 0>}]
 dyn_sh_bow_chest_l_1 [{shapeType=Capsule, length=3.2298, radius=1.3007, shapeName=dyn_sh_bow_chest_l_1_capsule, center=<2 -0.5 0>}]
 dyn_sh_bow_chest_l_2 [{shapeType=Capsule, length=3.4628, radius=1.3372, shapeName=dyn_sh_bow_chest_l_2_capsule, center=<2 0 0>, rotation=<rot 0 0 4.0623>}]
 spine_05 [{shapeType=Capsule, length=11.2687, radius=8.9845, shapeName=spine_05_capsule, center=<-3.9018 2.1594 0>}]
}
constraints [
 {bone1Name=dyn_bow_chest_r_0, bone2Name=base_bow_chest_r_0, swing1LimitDegrees=8, swing2LimitDegrees=5, twistLimitDegrees=0}
 {bone1Name=dyn_bow_chest_r_1, bone2Name=dyn_bow_chest_r_0, swing1LimitDegrees=8, swing2LimitDegrees=5, twistLimitDegrees=0}
 {bone1Name=dyn_bow_chest_r_2, bone2Name=dyn_bow_chest_r_1, swing1LimitDegrees=8, swing2LimitDegrees=5, twistLimitDegrees=0}
 {bone1Name=dyn_lrg_bow_chest_l_0, bone2Name=base_lrg_bow_chest_l_0, swing1LimitDegrees=5, swing2LimitDegrees=5, twistLimitDegrees=0}
 {bone1Name=dyn_lrg_bow_chest_l_1, bone2Name=dyn_lrg_bow_chest_l_0, swing1LimitDegrees=10, swing2LimitDegrees=10, twistLimitDegrees=0}
 {bone1Name=dyn_lrg_bow_chest_l_2, bone2Name=dyn_lrg_bow_chest_l_1, swing1LimitDegrees=10, swing2LimitDegrees=10, twistLimitDegrees=0}
 {bone1Name=dyn_lrg_bow_chest_r_0, bone2Name=base_lrg_bow_chest_r_0, swing1LimitDegrees=5, swing2LimitDegrees=5, twistLimitDegrees=0}
 {bone1Name=dyn_lrg_bow_chest_r_1, bone2Name=dyn_lrg_bow_chest_r_0, swing1LimitDegrees=10, swing2LimitDegrees=10, twistLimitDegrees=0}
 {bone1Name=dyn_lrg_bow_chest_r_2, bone2Name=dyn_lrg_bow_chest_r_1, swing1LimitDegrees=10, swing2LimitDegrees=10, twistLimitDegrees=0}
 {bone1Name=dyn_sh_bow_chest_l_0, bone2Name=base_sh_bow_chest_l_0, swing1LimitDegrees=8, swing2LimitDegrees=5, twistLimitDegrees=0}
 {bone1Name=dyn_sh_bow_chest_l_1, bone2Name=dyn_sh_bow_chest_l_0, swing1LimitDegrees=8, swing2LimitDegrees=5, twistLimitDegrees=0}
 {bone1Name=dyn_sh_bow_chest_l_2, bone2Name=dyn_sh_bow_chest_l_1, swing1LimitDegrees=8, swing2LimitDegrees=5, twistLimitDegrees=0}
]
