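# Control Rig for the BrightLarva female medium body cosmetic, correcting skirt, bows, arm feather and puff sleeves as the character moves.
# A shared DeformRig function runs first, then weighted parent and aim constraints position bows and feather.
# Spherical pose readers on the thighs push skirt bones, with a crouch pose weighted by pelvis height and isCrouching.
# Head rotation adjusts arm puffs via targeting and upper-arm pose, also setting the targeting_up curve.
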ControlRigBlueprint /BRCosmetics/Characters/Player/Female/Medium/Bodies/F_MED_BrightLarva/Meshes/F_MED_BrightLarva_CtrlRig
graphs [
 New Function_ContainedGraph {nodes [
   RigVMFunctionEntryNode()
   > RigVMFunction_Sequence = RigVMAggregateNode()
   n1 = &RigUnit_SphericalPoseReader(
    ActiveRegionScaleFactors=(PositiveWidth=1,NegativeWidth=1,PositiveHeight=1,NegativeHeight=1)
    ActiveRegionSize=0.05
    Debug=(bDrawDebug=false,bDraw2D=False,bDrawLocalAxes=False,DebugScale=25,DebugSegments=20,DebugThickness=0.25)
    DriverAxis=(X=-1,Y=0,Z=0)
    DriverItem=(Type=Bone,Name="thigh_l")
    FalloffRegionScaleFactors=(PositiveWidth=1,NegativeWidth=1,PositiveHeight=0.9,NegativeHeight=0.9)
    FalloffSize=0.63
    FlipHeightScaling=False
    FlipWidthScaling=False
    OptionalParentItem=(Type=Bone,Name="None")
    RotationOffset=(X=90,Y=0,Z=-30)
    ExecutePin=@RigVMFunction_Sequence.A
   )
   > &RigUnit_ModifyTransforms(
    ItemToModify="((Item=(Type=Bone,Name=\"base_skirt_front_l_0\"),Transform=(Rotation=(X=-0.262,Y=-0.2198,Z=-0.7198,W=0.604),Translation=(X=-5,Y=0,Z=0),Scale3D=(X=1,Y=1,Z=1))),(Item=(Type=Bone,Name=\"base_skirt_side_l_0\"),Transform=(Rotation=(X=0,Y=0,Z=0,W=1),Translation=(X=0 ...(truncated)"
    WeightMinimum=0
    Weight=@n1.OutputParam
   )
   > n2 = &RigUnit_SphericalPoseReader(
    ActiveRegionScaleFactors=(PositiveWidth=1,NegativeWidth=1,PositiveHeight=1,NegativeHeight=1)
    ActiveRegionSize=0.05
    Debug=(bDrawDebug=false,bDraw2D=False,bDrawLocalAxes=False,DebugScale=25,DebugSegments=20,DebugThickness=0.25)
    DriverAxis=(X=1,Y=0,Z=0)
    DriverItem=(Type=Bone,Name="thigh_r")
    FalloffRegionScaleFactors=(PositiveWidth=1,NegativeWidth=1,PositiveHeight=1,NegativeHeight=1)
    FalloffSize=0.63
    FlipHeightScaling=true
    FlipWidthScaling=true
    OptionalParentItem=(Type=Bone,Name="None")
    RotationOffset=(X=-90,Y=0,Z=-30)
   )
   > &RigUnit_ModifyTransforms(
    ItemToModify=((Item=(Type=Bone,Name="base_skirt_front_r_0"),Transform=(Rotation=(X=0.262,Y=0.2198,Z=-0.7198,W=0.604),Translation=(X=-5,Y=0,Z=0),Scale3D=(X=1,Y=1,Z=1))))
    WeightMinimum=0
    Weight=@n2.OutputParam
   )
   Maximum = &RigVMFunction_MathDoubleMax(B=@n1.OutputParam, A=@n2.OutputParam)
   > &RigUnit_ModifyTransforms(
    ItemToModify=((Item=(Type=Bone,Name="base_skirt_front_0"),Transform=(Rotation=(X=0,Y=0,Z=-0.766,W=0.6428),Translation=(X=-10,Y=0,Z=0),Scale3D=(X=1,Y=1,Z=1))))
    WeightMinimum=0
    Weight=@Maximum.Result
   )
   > n3 = &RigUnit_SphericalPoseReader(
    ActiveRegionScaleFactors=(PositiveWidth=1,NegativeWidth=1,PositiveHeight=1,NegativeHeight=1)
    ActiveRegionSize=0.05
    Debug=(bDrawDebug=false,bDraw2D=False,bDrawLocalAxes=False,DebugScale=25,DebugSegments=20,DebugThickness=0.25)
    DriverAxis=(X=-1,Y=0,Z=0)
    DriverItem=(Type=Bone,Name="thigh_l")
    FalloffRegionScaleFactors=(PositiveWidth=1,NegativeWidth=1,PositiveHeight=0.9,NegativeHeight=0.9)
    FalloffSize=0.63
    FlipHeightScaling=False
    FlipWidthScaling=False
    OptionalParentItem=(Type=Bone,Name="None")
    RotationOffset=(X=180,Y=30,Z=0)
   )
   > &RigUnit_ModifyTransforms(
    ItemToModify="((Item=(Type=Bone,Name=\"base_skirt_side_l_0\"),Transform=(Rotation=(X=0,Y=0,Z=-0.7934,W=0.6088),Translation=(X=0,Y=0,Z=0),Scale3D=(X=1,Y=1,Z=1))),(Item=(Type=Bone,Name=\"base_skirt_front_l_0\"),Transform=(Rotation=(X=-0.1172,Y=0.5288,Z=-0.2907,W=0.7887),Trans ...(truncated)"
    WeightMinimum=0
    Weight=@n3.OutputParam
   )
   > n4 = &RigUnit_SphericalPoseReader(
    ActiveRegionScaleFactors=(PositiveWidth=1,NegativeWidth=1,PositiveHeight=1,NegativeHeight=1)
    ActiveRegionSize=0.05
    Debug=(bDrawDebug=false,bDraw2D=False,bDrawLocalAxes=False,DebugScale=25,DebugSegments=20,DebugThickness=0.25)
    DriverAxis=(X=-1,Y=0,Z=0)
    DriverItem=(Type=Bone,Name="thigh_l")
    FalloffRegionScaleFactors=(PositiveWidth=1,NegativeWidth=1,PositiveHeight=0.9,NegativeHeight=0.9)
    FalloffSize=0.63
    FlipHeightScaling=False
    FlipWidthScaling=False
    OptionalParentItem=(Type=Bone,Name="None")
    RotationOffset=(X=140,Y=30,Z=0)
   )
   > &RigUnit_OffsetTransformForItem(
    Item=(Type=Bone,Name="base_skirt_front_l_0")
    OffsetTransform=(Rotation=(X=0.0227,Y=0.1722,Z=-0.1285,W=0.9764),Translation=(X=0,Y=0,Z=0),Scale3D=(X=1,Y=1,Z=1))
    bPropagateToChildren=True
    Weight=@n4.OutputParam
   )
   > n5 = &RigUnit_SphericalPoseReader(
    ActiveRegionScaleFactors=(PositiveWidth=1,NegativeWidth=1,PositiveHeight=1,NegativeHeight=1)
    ActiveRegionSize=0.05
    Debug=(bDrawDebug=false,bDraw2D=False,bDrawLocalAxes=False,DebugScale=25,DebugSegments=20,DebugThickness=0.25)
    DriverAxis=(X=1,Y=0,Z=0)
    DriverItem=(Type=Bone,Name="thigh_r")
    FalloffRegionScaleFactors=(PositiveWidth=1,NegativeWidth=1,PositiveHeight=0.9,NegativeHeight=0.9)
    FalloffSize=0.63
    FlipHeightScaling=False
    FlipWidthScaling=False
    OptionalParentItem=(Type=Bone,Name="None")
    RotationOffset=(X=0,Y=30,Z=0)
   )
   > &RigUnit_ModifyTransforms(
    ItemToModify="((Item=(Type=Bone,Name=\"base_skirt_side_r_0\"),Transform=(Rotation=(X=0,Y=0,Z=-0.7934,W=0.6088),Translation=(X=0,Y=0,Z=0),Scale3D=(X=1,Y=1,Z=1))),(Item=(Type=Bone,Name=\"base_skirt_front_r_0\"),Transform=(Rotation=(X=0.1172,Y=-0.5288,Z=-0.2907,W=0.7887),Trans ...(truncated)"
    WeightMinimum=0
    Weight=@n5.OutputParam
   )
   > n6 = &RigUnit_SphericalPoseReader(
    ActiveRegionScaleFactors=(PositiveWidth=1,NegativeWidth=1,PositiveHeight=1,NegativeHeight=1)
    ActiveRegionSize=0.05
    Debug=(bDrawDebug=false,bDraw2D=False,bDrawLocalAxes=False,DebugScale=25,DebugSegments=20,DebugThickness=0.25)
    DriverAxis=(X=1,Y=0,Z=0)
    DriverItem=(Type=Bone,Name="thigh_r")
    FalloffRegionScaleFactors=(PositiveWidth=1,NegativeWidth=1,PositiveHeight=0.9,NegativeHeight=0.9)
    FalloffSize=0.63
    FlipHeightScaling=False
    FlipWidthScaling=False
    OptionalParentItem=(Type=Bone,Name="None")
    RotationOffset=(X=-40,Y=30,Z=0)
   )
   > &RigUnit_OffsetTransformForItem(
    Item=(Type=Bone,Name="base_skirt_front_r_0")
    OffsetTransform=(Rotation=(X=-0.0227,Y=-0.1722,Z=-0.1285,W=0.9764),Translation=(X=0,Y=0,Z=0),Scale3D=(X=1,Y=1,Z=1))
    bPropagateToChildren=True
    Weight=@n6.OutputParam
   )
   > n7 = &RigUnit_SphericalPoseReader(
    ActiveRegionScaleFactors=(PositiveWidth=1,NegativeWidth=1,PositiveHeight=1,NegativeHeight=1)
    ActiveRegionSize=0.05
    Debug=(bDrawDebug=false,bDraw2D=False,bDrawLocalAxes=False,DebugScale=25,DebugSegments=20,DebugThickness=0.25)
    DriverAxis=(X=-1,Y=0,Z=0)
    DriverItem=(Type=Bone,Name="thigh_l")
    FalloffRegionScaleFactors=(PositiveWidth=0.75,NegativeWidth=0.75,PositiveHeight=0.9,NegativeHeight=0.9)
    FalloffSize=0.63
    FlipHeightScaling=False
    FlipWidthScaling=False
    OptionalParentItem=(Type=Bone,Name="None")
    RotationOffset=(X=-90,Y=0,Z=10)
   )
   > &RigUnit_ModifyTransforms(
    ItemToModify=((Item=(Type=Bone,Name="base_skirt_back_side_l_0"),Transform=(Rotation=(X=-0.1506,Y=0.1656,Z=-0.5126,W=0.8289),Translation=(X=0,Y=0,Z=0),Scale3D=(X=1,Y=1,Z=1))))
    WeightMinimum=0
    Weight=@n7.OutputParam
   )
   > n8 = &RigUnit_SphericalPoseReader(
    ActiveRegionScaleFactors=(PositiveWidth=1,NegativeWidth=1,PositiveHeight=1,NegativeHeight=1)
    ActiveRegionSize=0.05
    Debug=(bDrawDebug=false,bDraw2D=False,bDrawLocalAxes=False,DebugScale=25,DebugSegments=20,DebugThickness=0.25)
    DriverAxis=(X=1,Y=0,Z=0)
    DriverItem=(Type=Bone,Name="thigh_r")
    FalloffRegionScaleFactors=(PositiveWidth=0.75,NegativeWidth=0.75,PositiveHeight=0.9,NegativeHeight=0.9)
    FalloffSize=0.63
    FlipHeightScaling=False
    FlipWidthScaling=False
    OptionalParentItem=(Type=Bone,Name="None")
    RotationOffset=(X=90,Y=0,Z=10)
   )
   > &RigUnit_ModifyTransforms(
    ItemToModify=((Item=(Type=Bone,Name="base_skirt_back_side_r_0"),Transform=(Rotation=(X=0.1506,Y=-0.1656,Z=-0.5126,W=0.8289),Translation=(X=0,Y=0,Z=0),Scale3D=(X=1,Y=1,Z=1))))
    WeightMinimum=0
    Weight=@n8.OutputParam
   )
   > &RigUnit_ModifyTransforms(
    ItemToModify=((Item=(Type=Bone,Name="base_skirt_back_0"),Transform=(Rotation=(X=0,Y=0,Z=-0.766,W=0.6428),Translation=(X=0,Y=0,Z=0),Scale3D=(X=1,Y=1,Z=1))))
    WeightMinimum=0
    Weight=&RigVMFunction_MathDoubleMax(A=@n7.OutputParam, B=@n8.OutputParam).Result
   )
   > n9 = &RigUnit_SphericalPoseReader(
    ActiveRegionScaleFactors=(PositiveWidth=1,NegativeWidth=1,PositiveHeight=1,NegativeHeight=1)
    ActiveRegionSize=0.05
    Debug=(bDrawDebug=false,bDraw2D=False,bDrawLocalAxes=False,DebugScale=25,DebugSegments=20,DebugThickness=0.25)
    DriverAxis=(X=-1,Y=0,Z=0)
    DriverItem=(Type=Bone,Name="thigh_l")
    FalloffRegionScaleFactors=(PositiveWidth=1,NegativeWidth=1,PositiveHeight=0.9,NegativeHeight=0.9)
    FalloffSize=0.63
    FlipHeightScaling=False
    FlipWidthScaling=false
    OptionalParentItem=(Type=Bone,Name="None")
    RotationOffset=(X=-135,Y=0,Z=0)
   )
   > &RigUnit_OffsetTransformForItem(
    Item=(Type=Bone,Name="base_skirt_back_side_l_0")
    OffsetTransform=(Rotation=(X=-0.084,Y=-0.2504,Z=-0.2147,W=0.9403),Translation=(X=0,Y=0,Z=0),Scale3D=(X=1,Y=1,Z=1))
    bPropagateToChildren=True
    Weight=@n9.OutputParam
   )
   > n10 = &RigUnit_SphericalPoseReader(
    ActiveRegionScaleFactors=(PositiveWidth=1,NegativeWidth=1,PositiveHeight=1,NegativeHeight=1)
    ActiveRegionSize=0.05
    Debug=(bDrawDebug=false,bDraw2D=False,bDrawLocalAxes=False,DebugScale=25,DebugSegments=20,DebugThickness=0.25)
    DriverAxis=(X=1,Y=0,Z=0)
    DriverItem=(Type=Bone,Name="thigh_r")
    FalloffRegionScaleFactors=(PositiveWidth=1,NegativeWidth=1,PositiveHeight=0.9,NegativeHeight=0.9)
    FalloffSize=0.63
    FlipHeightScaling=False
    FlipWidthScaling=False
    OptionalParentItem=(Type=Bone,Name="None")
    RotationOffset=(X=45,Y=0,Z=0)
   )
   > &RigUnit_OffsetTransformForItem(
    Item=(Type=Bone,Name="base_skirt_back_side_r_0")
    OffsetTransform=(Rotation=(X=0.084,Y=0.2504,Z=-0.2147,W=0.9403),Translation=(X=0,Y=0,Z=0),Scale3D=(X=1,Y=1,Z=1))
    bPropagateToChildren=True
    Weight=@n10.OutputParam
   )
   > RigVMFunctionReturnNode()
   Remap = &RigVMFunction_MathDoubleRemap(
    SourceMaximum=45
    SourceMinimum=89
    TargetMaximum=1
    TargetMinimum=0
    Value=&RigUnit_GetRelativeTransformForItem(Child=(Type=Bone,Name="pelvis"), Parent=(Type=Bone,Name="root"), bChildInitial=False, bParentInitial=False).RelativeTransform.Translation.Z
   )
   RigVMFunction_MathBoolToFloat = &RigVMFunction_MathBoolToFloat(Value=RigVMVariableNode() {variable=isCrouching}.Value)
   Multiply = &RigVMFunction_MathDoubleMul(A=@Remap.Result, B=@RigVMFunction_MathBoolToFloat.Result)
   &RigUnit_ModifyTransforms(
    ItemToModify="((Item=(Type=Bone,Name=\"base_skirt_side_l_0\"),Transform=(Rotation=(X=-0.093,Y=0.4117,Z=0.1457,W=-0.8948),Translation=(X=0,Y=0,Z=0),Scale3D=(X=1,Y=1,Z=1))),(Item=(Type=Bone,Name=\"base_skirt_back_side_l_0\"),Transform=(Rotation=(X=0.0974,Y=0.2069,Z=-0.2957,W= ...(truncated)"
    WeightMinimum=0
    Weight=@Multiply.Result
    ExecutePin=@RigVMFunction_Sequence.B
   )
   > RigVMVariableNode(Value=@Multiply.Result) {variable=GetCrouchValue}
  ]}
 New Function_ContainedGraph {nodes [
   RigVMFunctionEntryNode()
   > &RigUnit_ParentConstraint(
    AdvancedSettings=(RotationOrderForFilter=XZY)
    Child=(Type=Bone,Name="base_bow")
    Filter=(TranslationFilter=(bX=True,bY=True,bZ=True),RotationFilter=(bX=true,bY=true,bZ=true),ScaleFilter=(bX=True,bY=True,bZ=True))
    Parents=((Item=(Type=Bone,Name="spine_04"),Weight=0.0947),(Item=(Type=Bone,Name="spine_03"),Weight=0.1001),(Item=(Type=Bone,Name="spine_01"),Weight=0.1158),(Item=(Type=Bone,Name="pelvis"),Weight=0.2246),(Item=(Type=Bone,Name="spine_02"),Weight=0.4648))
    Weight=1
   )
   > &RigUnit_ParentConstraint(
    AdvancedSettings=(RotationOrderForFilter=XZY)
    Child=(Type=Bone,Name="base_sh_bow_chest_l_0")
    Parents="((Item=(Type=Bone,Name=\"head\"),Weight=0.0134),(Item=(Type=Bone,Name=\"clavicle_l\"),Weight=0.0153),(Item=(Type=Bone,Name=\"spine_03\"),Weight=0.0162),(Item=(Type=Bone,Name=\"neck_02\"),Weight=0.0437),(Item=(Type=Bone,Name=\"neck_01\"),Weight=0.0661),(Item=(Type=Bo ...(truncated)"
    Weight=1
   )
   > &RigUnit_ParentConstraint(
    AdvancedSettings=(RotationOrderForFilter=XZY)
    Child=(Type=Bone,Name="base_bow_chest_r_0")
    Parents="((Item=(Type=Bone,Name=\"head\"),Weight=0.0134),(Item=(Type=Bone,Name=\"clavicle_r\"),Weight=0.0153),(Item=(Type=Bone,Name=\"spine_03\"),Weight=0.0162),(Item=(Type=Bone,Name=\"neck_02\"),Weight=0.0437),(Item=(Type=Bone,Name=\"neck_01\"),Weight=0.0661),(Item=(Type=Bo ...(truncated)"
    Weight=1
   )
   > &RigUnit_ParentConstraint(
    AdvancedSettings=(RotationOrderForFilter=XZY)
    Child=(Type=Bone,Name="base_lrg_bow_chest_r_0")
    Parents="((Item=(Type=Bone,Name=\"head\"),Weight=0.0134),(Item=(Type=Bone,Name=\"clavicle_r\"),Weight=0.0153),(Item=(Type=Bone,Name=\"spine_03\"),Weight=0.0162),(Item=(Type=Bone,Name=\"neck_02\"),Weight=0.0437),(Item=(Type=Bone,Name=\"neck_01\"),Weight=0.0661),(Item=(Type=Bo ...(truncated)"
    Weight=1
   )
   > &RigUnit_ParentConstraint(
    AdvancedSettings=(RotationOrderForFilter=XZY)
    Child=(Type=Bone,Name="base_lrg_bow_chest_l_0")
    Parents="((Item=(Type=Bone,Name=\"head\"),Weight=0.0134),(Item=(Type=Bone,Name=\"clavicle_l\"),Weight=0.0153),(Item=(Type=Bone,Name=\"spine_03\"),Weight=0.0162),(Item=(Type=Bone,Name=\"neck_02\"),Weight=0.0437),(Item=(Type=Bone,Name=\"neck_01\"),Weight=0.0661),(Item=(Type=Bo ...(truncated)"
    Weight=1
   )
   > &RigUnit_ParentConstraint(
    AdvancedSettings=(RotationOrderForFilter=XZY)
    Child=(Type=Bone,Name="for_aim_sh_bow_chest_l")
    Parents="((Item=(Type=Bone,Name=\"deform_upperarm_fr_l\"),Weight=0.0008),(Item=(Type=Bone,Name=\"clavicle_l\"),Weight=0.0287),(Item=(Type=Bone,Name=\"spine_05\"),Weight=0.2272),(Item=(Type=Bone,Name=\"deform_pec_l\"),Weight=0.3498),(Item=(Type=Bone,Name=\"spine_04\"),Weight= ...(truncated)"
    Weight=1
   )
   > &RigUnit_ParentConstraint(
    AdvancedSettings=(RotationOrderForFilter=XZY)
    Child=(Type=Bone,Name="for_aim_sh_bow_chest_r")
    Parents="((Item=(Type=Bone,Name=\"deform_upperarm_fr_r\"),Weight=0.0008),(Item=(Type=Bone,Name=\"clavicle_r\"),Weight=0.0287),(Item=(Type=Bone,Name=\"spine_05\"),Weight=0.2272),(Item=(Type=Bone,Name=\"deform_pec_r\"),Weight=0.3498),(Item=(Type=Bone,Name=\"spine_04\"),Weight= ...(truncated)"
    Weight=1
   )
   > RigVMFunctionReturnNode()
  ]}
 New Function_ContainedGraph {nodes [
   RigVMFunctionEntryNode()
   > &RigUnit_AimConstraintLocalSpaceOffset(
    Child=(Type=Bone,Name="base_sh_bow_chest_l_0")
    Filter=(bX=True,bY=True,bZ=True)
    Parents=((Item=(Type=Bone,Name="for_aim_sh_bow_chest_l"),Weight=1))
    UpAxis=(X=0,Y=1,Z=0)
    Weight=1
    WorldUp=(Target=(X=0,Y=1,Z=0),Kind=Location,Space=(Type=Bone,Name="for_aim_sh_bow_chest_l"))
   )
   > &RigUnit_RotationConstraintLocalSpaceOffset(
    AdvancedSettings=(InterpolationType=Average,RotationOrderForFilter=XZY)
    Child=(Type=Bone,Name="base_lrg_bow_chest_l_0")
    Filter=(bX=True,bY=True,bZ=True)
    Parents=((Item=(Type=Bone,Name="base_sh_bow_chest_l_0"),Weight=1))
    Weight=1
   )
   > &RigUnit_AimConstraintLocalSpaceOffset(
    Child=(Type=Bone,Name="base_bow_chest_r_0")
    Filter=(bX=True,bY=True,bZ=True)
    Parents=((Item=(Type=Bone,Name="for_aim_sh_bow_chest_r"),Weight=1))
    UpAxis=(X=0,Y=1,Z=0)
    Weight=1
    WorldUp=(Target=(X=0,Y=1,Z=0),Kind=Location,Space=(Type=Bone,Name="for_aim_sh_bow_chest_r"))
   )
   > &RigUnit_RotationConstraintLocalSpaceOffset(
    AdvancedSettings=(InterpolationType=Average,RotationOrderForFilter=XZY)
    Child=(Type=Bone,Name="base_lrg_bow_chest_r_0")
    Filter=(bX=True,bY=True,bZ=True)
    Parents=((Item=(Type=Bone,Name="base_bow_chest_r_0"),Weight=1))
    Weight=1
   )
   > RigVMFunctionReturnNode()
  ]}
 New Function_ContainedGraph {nodes [
   n1 = RigVMFunctionEntryNode()
   RigVMFunction_MathBoolToInteger = &RigVMFunction_MathBoolToInteger(Value=@n1.onlyTwist)
   RigVMFunction_MathQuaternionSwingTwist = &RigVMFunction_MathQuaternionSwingTwist(
    twistAxis=@n1.twistAxis
    Input=&RigUnit_GetRelativeTransformForItem(
     bChildInitial=False
     bParentInitial=False
     child=(Type=Bone,Name="None")
     parent=(Type=Bone,Name="None")
     child.Name=@n1.child
     parent.Name=@n1.parent
     parent.Type=@n1.parentType
    ).RelativeTransform.Rotation
   )
   RigVMFunction_MathQuaternionToEuler = &RigVMFunction_MathQuaternionToEuler(
    Value=RigVMDispatchNode(
     Values=((),())
     Index=@RigVMFunction_MathBoolToInteger.Result
     Values.0=@RigVMFunction_MathQuaternionSwingTwist.Swing
     Values.1=@RigVMFunction_MathQuaternionSwingTwist.Twist
    ) {template="DISPATCH_RigVMDispatch_SelectInt32(in Index,in Values,out Result)"}.Result
    RotationOrder=@n1.Rotation Order
   )
   > DISPATCH_RigVMDispatch_Print = RigVMDispatchNode(ScreenColor=(R=1,G=1,B=1,A=1), ScreenDuration=1.5, Enabled=@n1.debugRotation, Value=@RigVMFunction_MathQuaternionToEuler.Result) {template="DISPATCH_RigVMDispatch_Print(in Prefix,in Value,in Enabled,in ScreenDuration,in ScreenColor)"}
   Clamp = &RigVMFunction_MathIntClamp(Value=@n1.primAxis)
   DISPATCH_RigVMDispatch_SelectInt32 = RigVMDispatchNode(
    Values=(0,0,0)
    Index=@Clamp.Result
    Values.0=@RigVMFunction_MathQuaternionToEuler.Result.X
    Values.1=@RigVMFunction_MathQuaternionToEuler.Result.Y
    Values.2=@RigVMFunction_MathQuaternionToEuler.Result.Z
   ) {template="DISPATCH_RigVMDispatch_SelectInt32(in Index,in Values,out Result)"}
   Remap = &RigVMFunction_MathDoubleRemap(
    TargetMaximum=1
    TargetMinimum=0
    bClamp=False
    Value=@DISPATCH_RigVMDispatch_SelectInt32.Result
    SourceMaximum=@n1.max
    SourceMinimum=@n1.min
   )
   Multiply = &RigVMFunction_MathDoubleMul(A=&RigVMFunction_MathDoubleClamp(Value=@Remap.Result).Result, B=@n1.premult)
   > RigVMFunctionReturnNode(outputWeight=@Multiply.Result)
  ]}
 New Function_ContainedGraph {nodes [
   RigVMFunctionEntryNode()
   > &RigUnit_ParentConstraint(
    AdvancedSettings=(RotationOrderForFilter=XZY)
    Child=(Type=Bone,Name="dyn_arm_feather_for_aim")
    Filter=(TranslationFilter=(bX=True,bY=True,bZ=True),RotationFilter=(bX=false,bY=false,bZ=false),ScaleFilter=(bX=True,bY=True,bZ=True))
    Parents="((Item=(Type=Bone,Name=\"upperarm_r\"),Weight=0.0135),(Item=(Type=Bone,Name=\"upperarm_twist_02_r\"),Weight=0.0147),(Item=(Type=Bone,Name=\"spine_05\"),Weight=0.0172),(Item=(Type=Bone,Name=\"deform_upperarm_fr_r\"),Weight=0.0318),(Item=(Type=Bone,Name=\"deform_uppe ...(truncated)"
    Weight=1
   )
   > &RigUnit_ParentConstraint(
    AdvancedSettings=(RotationOrderForFilter=XZY)
    Child=(Type=Bone,Name="base_arm_feather_r")
    Parents="((Item=(Type=Bone,Name=\"upperarm_r\"),Weight=0.0112),(Item=(Type=Bone,Name=\"upperarm_twist_02_r\"),Weight=0.0121),(Item=(Type=Bone,Name=\"spine_05\"),Weight=0.0139),(Item=(Type=Bone,Name=\"deform_upperarm_bk_r\"),Weight=0.021),(Item=(Type=Bone,Name=\"deform_upper ...(truncated)"
    Weight=1
   )
   > &RigUnit_AimConstraintLocalSpaceOffset(
    Child=(Type=Bone,Name="base_arm_feather_r")
    Filter=(bX=True,bY=True,bZ=True)
    Parents=((Item=(Type=Bone,Name="dyn_arm_feather_for_aim"),Weight=1))
    Weight=1
    WorldUp=(Target=(X=0,Y=0,Z=1),Kind=Location,Space=(Type=Bone,Name="dyn_arm_feather_for_aim"))
   )
   > RigVMFunctionReturnNode()
  ]}
 New Function_ContainedGraph {nodes [
   RigVMFunctionEntryNode()
   > DeltaRotation_6 = RigVMFunctionReferenceNode(
    Rotation Order=ZYX
    child=head
    debugRotation=false
    max=-25
    min=0
    parent=head_null
    parentType=Null
    premult=1
    primAxis=2
    twistAxis=(X=1,Y=0,Z=0)
   )
   Cast = &RigVMFunction_MathBoolToFloat(Value=RigVMVariableNode() {variable=isTargeting}.Value)
   Multiply = &RigVMFunction_MathDoubleMul(B=@Cast.Result, A=@DeltaRotation_6.outputWeight)
   AlphaInterp = &RigVMFunction_AlphaInterp(InterpSpeedDecreasing=10, InterpSpeedIncreasing=10, Scale=1, bClampResult=false, bMapRange=false, Value=@Multiply.Result)
   n1 = &RigVMFunction_AnimEvalRichCurve(
    Curve="(EditorCurveData=(Keys=((InterpMode=RCIM_Cubic,TangentMode=RCTM_User,TangentWeightMode=RCTWM_WeightedNone,Time=0,Value=0,ArriveTangent=2.4061,ArriveTangentWeight=0,LeaveTangent=2.4061,LeaveTangentWeight=0),(InterpMode=RCIM_Cubic,TangentMode=RCTM_Auto,Tange ...(truncated)"
    SourceMinimum=0
    TargetMinimum=0
    Value=@AlphaInterp.Result
   )
   Subtract = &RigVMFunction_MathDoubleSub(B=1, A=@n1.Result)
   Absolute = &RigVMFunction_MathDoubleAbs(Value=@Subtract.Result)
   > /Script/ControlRig.RigUnit_SetCurveValue(Curve=targeting_up, Value=@Absolute.Result) {is=struct RigVMUnitNode}
   > &RigUnit_ModifyTransforms(
    ItemToModify="((Item=(Type=Bone,Name=\"dyn_arm_puff_r\"),Transform=(Rotation=(X=0,Y=0,Z=0,W=-1),Translation=(X=0,Y=0,Z=0),Scale3D=(X=1,Y=1,Z=1))),(Item=(Type=Bone,Name=\"upperarm_twist_01_r\"),Transform=(Rotation=(X=0.2588,Y=0,Z=0,W=0.9659),Translation=(X=0,Y=0,Z=0),Scale3D ...(truncated)"
    WeightMinimum=0
    Weight=@n1.Result
   )
   > n2 = &RigUnit_SphericalPoseReader(
    ActiveRegionScaleFactors=(PositiveWidth=1,NegativeWidth=1,PositiveHeight=1,NegativeHeight=1)
    ActiveRegionSize=0.07
    Debug=(bDrawDebug=false,bDraw2D=False,bDrawLocalAxes=False,DebugScale=25,DebugSegments=20,DebugThickness=0.25)
    DriverAxis=(X=-1,Y=0,Z=0)
    DriverItem=(Type=Bone,Name="upperarm_r")
    FalloffRegionScaleFactors=(PositiveWidth=1,NegativeWidth=1,PositiveHeight=1,NegativeHeight=1)
    FalloffSize=0.15
    FlipHeightScaling=False
    FlipWidthScaling=False
    OptionalParentItem=(Type=Bone,Name="None")
    RotationOffset=(X=-115,Y=0,Z=35)
   )
   > &RigUnit_ModifyTransforms(
    ItemToModify="((Item=(Type=Bone,Name=\"dyn_arm_feather_r\"),Transform=(Rotation=(X=0.0083,Y=0.1018,Z=0.0947,W=-0.9903),Translation=(X=0,Y=0,Z=0),Scale3D=(X=0.65,Y=0.65,Z=1))),(Item=(Type=Bone,Name=\"base_arm_puff_r\"),Transform=(Rotation=(X=-0.0419,Y=-0.1406,Z=-0.2824,W=0.9 ...(truncated)"
    WeightMinimum=0
    Weight=&RigVMFunction_MathDoubleMul(A=@n2.OutputParam, B=@DeltaRotation_6.outputWeight).Result
   )
   > RigVMFunctionReturnNode()
  ]}
 RigVMFunctionLibrary {nodes [
   DeltaRotation = RigVMCollapseNode(ExecuteContext=(), Rotation Order=ZYX, premult=1, twistAxis=(X=1,Y=0,Z=0))
   arm feather = RigVMCollapseNode(ExecuteContext=())
   chest bow = RigVMCollapseNode(ExecuteContext=())
   puff setup for hair = RigVMCollapseNode(ExecuteContext=())
   skirt = RigVMCollapseNode(ExecuteContext=())
   surface constraints = RigVMCollapseNode(ExecuteContext=())
  ]}
 RigVMFunction_Sequence_1_ContainedGraph {nodes [
   n1 = RigVMFunctionEntryNode()
   RigVMFunction_Sequence = &RigVMFunction_Sequence(ExecuteContext=@n1.ExecuteContext)
   n2 = &RigVMFunction_Sequence(ExecuteContext=@RigVMFunction_Sequence.B)
   RigVMFunctionReturnNode(A=@RigVMFunction_Sequence.A, B=@n2.A, C=@n2.B)
  ]}
 RigVMModel {nodes [
   RigUnit_BeginExecution = &RigUnit_BeginExecution()
   FN_DeformRig_FNC = RigVMFunctionReferenceNode(
    DeformRig_ Curve="(deformRig_l_upperarm_up_alpha_72_#1=False,deformRig_r_upperarm_up_alpha_73_#2=False,deformRig_l_upperarm_fr_alpha_74_#3=False,deformRig_r_upperarm_fr_alpha_75_CB51B8 ...(truncated)"
    DeformRig_Bool="(upperarm_1_#4=True,elbow_3_#5=True,wrist_7_#6=True,groin_8_#7=True,glute_17_#8=True,knee_9_40C408324CA261 ...(truncated)"
    DeformRig_Debug=(upperarm_debug_18_#4=False,wrist_debug_20_#6=False,lat_debug_26_#9=False,groinGlute_debug_27_#7=False)
    DeformRig_Value="(upperarm_up_18_#10=(Rotation=(X=0.0,Y=-0.0,Z=0.0,W=1.0),Translation=(X=0.0,Y=0.0,Z=8.0),Scale3D=(X=0.70,Y=1.0,Z=1.0)),upperarm_fwd_fr_19_#11=(Rotation=(X=0.0,Y=-0.0,Z=0.0,W=1.0),Translation=(X=0.0, ...(truncated)"
    ExecuteContext=@RigUnit_BeginExecution.ExecutePin
   )
   > surface constraints = RigVMFunctionReferenceNode()
   > chest bow = RigVMFunctionReferenceNode()
   > skirt_1 = RigVMFunctionReferenceNode()
   > arm feather = RigVMFunctionReferenceNode()
   > puff setup for hair = RigVMFunctionReferenceNode()
  ]}
]
hierarchy [
 root {loc=<0 0 0>, rot=<0 0 0 1>}
 pelvis {parent=root, loc=<0 -0.5301 98.0571>, rot=<0.0127 0.707 -0.0127 -0.707>}
 spine_01 {parent=pelvis, loc=<6.1208 -0.9328 0>, rot=<0 0 0.034 -0.9994>}
 spine_02 {parent=spine_01, loc=<6.5997 -0.2095 0>, rot=<0 0 0.0238 -0.9997>}
 spine_03 {parent=spine_02, loc=<6.4914 -0.1237 0>, rot=<0 0 0.0149 -0.9999>}
 spine_04 {parent=spine_03, loc=<6.9598 0.0794 0>, rot=<0 0 -0.0132 -0.9999>}
 spine_05 {parent=spine_04, loc=<6.9987 0 0>, rot=<0 0 -0.0015 -1>}
 neck_01 {parent=spine_05, loc=<11.3138 0.1529 0>, rot=<0 0 -0.2003 -0.9797>}
 neck_02 {parent=neck_01, loc=<4.5627 0 0>, rot=<0 0 0.0504 -0.9987>}
 head {parent=neck_02, loc=<5.3914 0 0>, rot=<0 0 0.115 -0.9934>}
 FACIAL_C_FacialRoot {parent=head, loc=<1.7294 0.1905 0.0003>, rot=<0.5021 -0.4979 -0.5021 -0.4979>}
 FACIAL_C_UnderChin {parent=FACIAL_C_FacialRoot, loc=<-0.0003 8.095 3.2775>, rot=<0.5 0 0 -0.866>}
 FACIAL_L_UnderChin {parent=FACIAL_C_FacialRoot, loc=<2.4314 7.0451 2.2776>, rot=<0.4025 -0.355 0.0823 -0.8398>}
 FACIAL_R_UnderChin {parent=FACIAL_C_FacialRoot, loc=<-2.4314 7.0451 2.2776>, rot=<0.4025 0.355 -0.0823 -0.8398>}
 FACIAL_C_Jaw {parent=FACIAL_C_FacialRoot, loc=<0.0003 1.9225 1.5671>, rot=<0.342 0 0 -0.9397>}
 FACIAL_C_Jawline {parent=FACIAL_C_Jaw, loc=<-0.0004 2.035 7.3789>, rot=<0.2588 0 0 -0.9659>}
 FACIAL_L_ChinSide {parent=FACIAL_C_Jaw, loc=<2.2791 0.0107 7.1997>, rot=<-0.2207 -0.3739 0.05 -0.8994>}
 FACIAL_R_ChinSide {parent=FACIAL_C_Jaw, loc=<-2.2791 0.0107 7.1997>, rot=<-0.2207 0.3739 -0.05 -0.8994>}
 FACIAL_C_Chin {parent=FACIAL_C_Jaw, loc=<0.0001 0.6106 8.564>, rot=<-0.2588 0 0 -0.9659>}
 head_null {parent=neck_02, type=Null, loc=<5.3914 0 0>, rot=<0 0 0.115 -0.9934>}
 clavicle_l {parent=spine_05, loc=<5.4831 -0.1053 -3.2946>, rot=<0.0465 -0.7289 -0.009 -0.6829>}
 upperarm_l {parent=clavicle_l, loc=<9.8658 -1.2136 -0.4105>, rot=<0.0164 -0.3245 -0.0269 -0.9454>}
 lowerarm_l {parent=upperarm_l, loc=<26.465 -0.8345 0.259>, rot=<0.0036 0.0161 -0.2298 -0.9731>}
 hand_l {parent=lowerarm_l, loc=<22.7355 -0.3203 -0.6105>, rot=<-0.6784 -0.0223 0.0735 -0.7307>}
 thumb_01_l {parent=hand_l, loc=<1.3118 -0.3191 -2.459>, rot=<0.6546 -0.4042 0.0664 -0.6354>}
 thumb_02_l {parent=thumb_01_l, loc=<5.9336 -0.0448 0.0401>, rot=<-0.0037 0.0213 0.298 -0.9543>}
 thumb_03_l {parent=thumb_02_l, loc=<2.2223 0.3682 -0.0907>, rot=<0.0118 -0.013 0.0066 -0.9998>}
 index_metacarpal_l {parent=hand_l, loc=<1.6535 0.0183 -2.96>, rot=<0.0158 -0.0124 -0.0173 -0.9996>}
 index_01_l {parent=index_metacarpal_l, loc=<8.3008 0.0357 -0.0524>, rot=<-0.01 0.0476 0.2089 -0.9767>}
 index_02_l {parent=index_01_l, loc=<3.3587 -0.3004 0.0452>, rot=<-0.002 0.0027 0.3225 -0.9466>}
 index_03_l {parent=index_02_l, loc=<2.5578 -0.123 -0.0252>, rot=<-0.002 0.0067 0.2558 -0.9667>}
 middle_metacarpal_l {parent=hand_l, loc=<2.1856 -0.1725 -0.5984>, rot=<0.006 0.0168 -0.0329 -0.9993>}
 middle_01_l {parent=middle_metacarpal_l, loc=<7.3275 -0.2447 0.0129>, rot=<-0.0208 0.0042 0.0858 -0.9961>}
 middle_02_l {parent=middle_01_l, loc=<4.0123 0.1401 0.1175>, rot=<-0.0218 0.0258 0.3512 -0.9357>}
 middle_03_l {parent=middle_02_l, loc=<2.6077 0.0534 0.1373>, rot=<0.0082 0.0188 0.1708 -0.9851>}
 ring_metacarpal_l {parent=hand_l, loc=<2.2074 -0.3625 1.3638>, rot=<-0.0284 0.0722 -0.0458 -0.9959>}
 ring_01_l {parent=ring_metacarpal_l, loc=<7.0295 -0.6197 -0.1446>, rot=<-0.0489 -0.0252 0.1892 -0.9804>}
 ring_02_l {parent=ring_01_l, loc=<3.6174 0.4647 -0.1027>, rot=<0 0 0.2806 -0.9598>}
 ring_03_l {parent=ring_02_l, loc=<2.4152 0.114 0.0199>, rot=<0 0 0.1189 -0.9929>}
 pinky_metacarpal_l {parent=hand_l, loc=<1.7616 -0.9753 3.3814>, rot=<-0.0397 0.0814 -0.069 -0.9935>}
 pinky_01_l {parent=pinky_metacarpal_l, loc=<7.1709 -0.4831 -0.1094>, rot=<-0.0654 -0.0035 0.2242 -0.9723>}
 pinky_02_l {parent=pinky_01_l, loc=<2.8274 0.0893 -0.1043>, rot=<0.0037 0.0011 0.2854 -0.9584>}
 pinky_03_l {parent=pinky_02_l, loc=<2.2046 0.0307 -0.0078>, rot=<-0.0002 0.0098 0.1445 -0.9895>}
 deform_wrist_dn_base_l {parent=hand_l, loc=<0 -1.5326 0>, rot=<0 0 0 -1>}
 deform_wrist_dn_l {parent=deform_wrist_dn_base_l, loc=<0 0 0>, rot=<0 0 0 1>}
 deform_wrist_up_base_l {parent=hand_l, loc=<0 2.3009 0>, rot=<0 0 0 -1>}
 deform_wrist_up_l {parent=deform_wrist_up_base_l, loc=<0 0 0>, rot=<0 0 0 1>}
 lowerarm_twist_02_l {parent=lowerarm_l, loc=<10.6707 0 0>, rot=<0 0 0 1>}
 lowerarm_twist_01_l {parent=lowerarm_l, loc=<18.316 0 0>, rot=<0 0 0 1>}
 deform_elbow_in_base_l {parent=lowerarm_l, loc=<0.1834 3.6062 -0.2572>, rot=<0 0 0 -1>}
 deform_elbow_in_l {parent=deform_elbow_in_base_l, loc=<0 0 0>, rot=<0 0 0 1>}
 deform_elbow_out_base_l {parent=lowerarm_l, loc=<0 -4 0>, rot=<0 0 0 1>}
 deform_elbow_out_l {parent=deform_elbow_out_base_l, loc=<0 0 0>, rot=<0 0 0 1>}
 upperarm_twist_01_l {parent=upperarm_l, loc=<7.2135 -0.1134 0>, rot=<0 0 0 1>}
 deform_upperarm_bk_base_l {parent=upperarm_twist_01_l, loc=<-6.4053 -4.3746 0.314>, rot=<0 0 0 -1>}
 deform_upperarm_bk_l {parent=deform_upperarm_bk_base_l, loc=<0 0 0>, rot=<0 0 0 1>}
 deform_upperarm_fr_base_l {parent=upperarm_twist_01_l, loc=<-5.6312 5.092 0.3354>, rot=<0 0 0 -1>}
 deform_upperarm_fr_l {parent=deform_upperarm_fr_base_l, loc=<0 0 0>, rot=<0 0 0 1>}
 deform_upperarm_up_base_l {parent=upperarm_twist_01_l, loc=<-8.891 0.4825 3.9517>, rot=<0 0 0 -1>}
 deform_upperarm_up_l {parent=deform_upperarm_up_base_l, loc=<0 0 0>, rot=<0 0 0 1>}
 base_arm_puff_l {parent=deform_upperarm_up_base_l, loc=<1.6775 -0.3691 -3.9517>, rot=<0 0.7071 0 -0.7071>}
 dyn_arm_puff_l {parent=base_arm_puff_l, loc=<0 0 0>, rot=<0 0 0 1>}
 upperarm_twist_02_l {parent=upperarm_l, loc=<19.7998 -0.1134 0>, rot=<0 0 0 1>}
 clavicle_r {parent=spine_05, loc=<5.4831 -0.1053 3.2946>, rot=<0.7289 0.0465 -0.6829 0.009>}
 upperarm_r {parent=clavicle_r, loc=<-9.8658 1.2136 0.4105>, rot=<0.0164 -0.3245 -0.0269 -0.9454>}
 lowerarm_r {parent=upperarm_r, loc=<-26.465 0.8345 -0.259>, rot=<0.0036 0.0161 -0.2298 -0.9731>}
 hand_r {parent=lowerarm_r, loc=<-22.7355 0.3203 0.6106>, rot=<-0.6784 -0.0223 0.0735 -0.7307>}
 thumb_01_r {parent=hand_r, loc=<-1.3118 0.3191 2.459>, rot=<0.6546 -0.4042 0.0664 -0.6354>}
 thumb_02_r {parent=thumb_01_r, loc=<-5.9336 0.0448 -0.0401>, rot=<-0.0037 0.0213 0.298 -0.9543>}
 thumb_03_r {parent=thumb_02_r, loc=<-2.2223 -0.3682 0.0907>, rot=<0.0118 -0.013 0.0066 -0.9998>}
 index_metacarpal_r {parent=hand_r, loc=<-1.6535 -0.0183 2.96>, rot=<0.0158 -0.0124 -0.0173 -0.9996>}
 index_01_r {parent=index_metacarpal_r, loc=<-8.3008 -0.0357 0.0524>, rot=<-0.0044 0.0485 0.0924 -0.9945>}
 index_02_r {parent=index_01_r, loc=<-3.3587 0.3004 -0.0452>, rot=<-0.002 0.0027 0.3225 -0.9466>}
 index_03_r {parent=index_02_r, loc=<-2.5578 0.123 0.0252>, rot=<-0.002 0.0067 0.2558 -0.9667>}
 middle_metacarpal_r {parent=hand_r, loc=<-2.1856 0.1725 0.5984>, rot=<0.006 0.0168 -0.0329 -0.9993>}
 middle_01_r {parent=middle_metacarpal_r, loc=<-7.3275 0.2447 -0.0129>, rot=<-0.0208 0.0042 0.0858 -0.9961>}
 middle_02_r {parent=middle_01_r, loc=<-4.0123 -0.1401 -0.1175>, rot=<-0.0218 0.0258 0.3512 -0.9357>}
 middle_03_r {parent=middle_02_r, loc=<-2.6077 -0.0534 -0.1373>, rot=<0.0082 0.0188 0.1708 -0.9851>}
 ring_metacarpal_r {parent=hand_r, loc=<-2.2074 0.3625 -1.3638>, rot=<-0.0284 0.0722 -0.0458 -0.9959>}
 ring_01_r {parent=ring_metacarpal_r, loc=<-7.0295 0.6197 0.1446>, rot=<-0.0489 -0.0252 0.1892 -0.9804>}
 ring_02_r {parent=ring_01_r, loc=<-3.6174 -0.4647 0.1027>, rot=<0 0 0.2806 -0.9598>}
 ring_03_r {parent=ring_02_r, loc=<-2.4152 -0.114 -0.0199>, rot=<0 0 0.1189 -0.9929>}
 pinky_metacarpal_r {parent=hand_r, loc=<-1.7616 0.9753 -3.3814>, rot=<-0.0397 0.0814 -0.069 -0.9935>}
 pinky_01_r {parent=pinky_metacarpal_r, loc=<-7.1709 0.4831 0.1094>, rot=<-0.0654 -0.0035 0.2242 -0.9723>}
 pinky_02_r {parent=pinky_01_r, loc=<-2.8274 -0.0893 0.1043>, rot=<0.0037 0.0011 0.2854 -0.9584>}
 pinky_03_r {parent=pinky_02_r, loc=<-2.2046 -0.0307 0.0078>, rot=<-0.0002 0.0098 0.1445 -0.9895>}
 deform_wrist_dn_base_r {parent=hand_r, loc=<-0.0001 1.5328 0>, rot=<0 0 0 -1>}
 deform_wrist_dn_r {parent=deform_wrist_dn_base_r, loc=<0 0 0>, rot=<0 0 0 1>}
 deform_wrist_up_base_r {parent=hand_r, loc=<-0.0002 -2.3007 -0.0001>, rot=<0 0 0 -1>}
 deform_wrist_up_r {parent=deform_wrist_up_base_r, loc=<0 0 0>, rot=<0 0 0 1>}
 lowerarm_twist_02_r {parent=lowerarm_r, loc=<-10.6707 0 0>, rot=<0 0 0 -1>}
 lowerarm_twist_01_r {parent=lowerarm_r, loc=<-18.316 0 0>, rot=<0 0 0 -1>}
 deform_elbow_in_base_r {parent=lowerarm_r, loc=<-0.1834 -3.6062 0.2572>, rot=<0 0 0 -1>}
 deform_elbow_in_r {parent=deform_elbow_in_base_r, loc=<0 0 0>, rot=<0 0 0 1>}
 deform_elbow_out_base_r {parent=lowerarm_r, loc=<0 4 -0.0001>, rot=<0 0 0 -1>}
 deform_elbow_out_r {parent=deform_elbow_out_base_r, loc=<0 0 0>, rot=<0 0 0 1>}
 upperarm_twist_01_r {parent=upperarm_r, loc=<-7.2135 0.1134 0>, rot=<0 0 0 -1>}
 deform_upperarm_bk_base_r {parent=upperarm_twist_01_r, loc=<6.4053 4.3746 -0.314>, rot=<0 0 0 -1>}
 deform_upperarm_bk_r {parent=deform_upperarm_bk_base_r, loc=<0 0 0>, rot=<0 0 0 1>}
 deform_upperarm_fr_base_r {parent=upperarm_twist_01_r, loc=<5.6312 -5.092 -0.3354>, rot=<0 0 0 -1>}
 deform_upperarm_fr_r {parent=deform_upperarm_fr_base_r, loc=<0 0 0>, rot=<0 0 0 1>}
 deform_upperarm_up_base_r {parent=upperarm_twist_01_r, loc=<8.891 -0.4825 -3.9517>, rot=<0 0 0 -1>}
 deform_upperarm_up_r {parent=deform_upperarm_up_base_r, loc=<0 0 0>, rot=<0 0 0 1>}
 base_arm_puff_r {parent=deform_upperarm_up_base_r, loc=<-1.6773 0.3691 3.9513>, rot=<0 -0.7071 0 -0.7071>}
 dyn_arm_puff_r {parent=base_arm_puff_r, loc=<0 0 0>, rot=<0 0 0 -1>}
 base_arm_feather_r {parent=upperarm_twist_01_r, loc=<-0.4779 0.0648 -5.959>, rot=<0.8278 0.0031 -0.561 -0.0069>}
 dyn_arm_feather_r {parent=base_arm_feather_r, loc=<0 0 0>, rot=<0 0 0 1>}
 dyn_arm_feather_for_aim {parent=upperarm_twist_01_r, loc=<4.0654 1.6593 -12.8342>, rot=<0.8278 0.0034 -0.561 -0.0076>, scale=<0.9691 1 1.0001>}
 upperarm_twist_02_r {parent=upperarm_r, loc=<-19.7998 0.1134 0>, rot=<0 0 0 -1>}
 attach_backpack {parent=spine_05, loc=<-0.0871 -13.4805 0.0013>, rot=<-0.0092 -0.7071 0.0093 -0.707>, scale=<0.9999 0.9954 1.0048>}
 Backpack_BR {parent=attach_backpack, type=Null, loc=<0 4 0>, rot=<0 0 0 1>}
 attach_cape {parent=spine_05, loc=<9.0438 -11.7201 0.0009>, rot=<0.0617 -0.7045 -0.0616 -0.7044>, scale=<0.9999 0.9954 1.0048>}
 cape {parent=attach_cape, type=Null, loc=<0 1.5 0>, rot=<0 0 0 1>}
 base_sh_bow_chest_l_0 {parent=spine_05, loc=<7.0026 4.8375 -0.5396>, rot=<-0.1847 0.0393 0.9 0.3928>}
 dyn_sh_bow_chest_l_0 {parent=base_sh_bow_chest_l_0, loc=<0 0 0>, rot=<0 0 0 1>}
 dyn_sh_bow_chest_l_1 {parent=dyn_sh_bow_chest_l_0, loc=<3.5 0 0>, rot=<0 -0.0704 0 -0.9975>}
 dyn_sh_bow_chest_l_2 {parent=dyn_sh_bow_chest_l_1, loc=<3.5 0 0>, rot=<0 0 0 1>}
 base_bow_chest_r_0 {parent=spine_05, loc=<7.0029 4.8375 0.5396>, rot=<0.1847 -0.0393 0.9 0.3928>}
 dyn_bow_chest_r_0 {parent=base_bow_chest_r_0, loc=<0 0 0>, rot=<0 0 0 1>}
 dyn_bow_chest_r_1 {parent=dyn_bow_chest_r_0, loc=<3.5001 0.0002 -0.0001>, rot=<0 0.0704 0 -0.9975>}
 dyn_bow_chest_r_2 {parent=dyn_bow_chest_r_1, loc=<3.5 0.0001 0>, rot=<0 0 0 1>}
 base_lrg_bow_chest_r_0 {parent=spine_05, loc=<6.7018 5.4128 0.6445>, rot=<0.0331 -0.0356 0.9434 0.3281>}
 dyn_lrg_bow_chest_r_0 {parent=base_lrg_bow_chest_r_0, loc=<0 0 0>, rot=<0 0 0 1>}
 dyn_lrg_bow_chest_r_1 {parent=dyn_lrg_bow_chest_r_0, loc=<5.5001 0.0001 0>, rot=<-0.0029 0.0654 -0.0755 -0.995>}
 dyn_lrg_bow_chest_r_2 {parent=dyn_lrg_bow_chest_r_1, loc=<5.4994 -0.0004 0>, rot=<-0.0302 0.171 -0.171 -0.9698>}
 base_lrg_bow_chest_l_0 {parent=spine_05, loc=<6.702 5.4128 -0.6445>, rot=<-0.0249 0.0327 0.9437 0.3284>}
 dyn_lrg_bow_chest_l_0 {parent=base_lrg_bow_chest_l_0, loc=<0 0 0>, rot=<0 0 0 1>}
 dyn_lrg_bow_chest_l_1 {parent=dyn_lrg_bow_chest_l_0, loc=<5.5 0 0>, rot=<-0.0005 -0.0202 -0.0756 -0.9969>}
 dyn_lrg_bow_chest_l_2 {parent=dyn_lrg_bow_chest_l_1, loc=<5.5 0 0>, rot=<0.0151 -0.0858 -0.173 -0.9811>}
 deform_pec_aim_base_l {parent=spine_04, loc=<3.2456 9.3309 -1.2237>, rot=<0.0603 -0.4033 0.3318 -0.8507>}
 deform_pec_aim_l {parent=deform_pec_aim_base_l, loc=<0 0 0>, rot=<0 0 0 1>}
 deform_pec_base_l {parent=deform_pec_aim_l, loc=<9.3665 0 0>, rot=<0 0 0 1>}
 deform_pec_l {parent=deform_pec_base_l, loc=<0 0 0>, rot=<0 0 0 1>}
 deform_pec_aim_base_r {parent=spine_04, loc=<3.2452 9.3309 1.2237>, rot=<-0.4033 -0.0603 0.8507 0.3318>}
 deform_pec_aim_r {parent=deform_pec_aim_base_r, loc=<0 0 0>, rot=<0 0 0 1>}
 deform_pec_base_r {parent=deform_pec_aim_r, loc=<-9.3667 -0.0003 -0.0004>, rot=<0 0 0 1>}
 deform_pec_r {parent=deform_pec_base_r, loc=<0 0 0>, rot=<0 0 0 1>}
 thigh_l {parent=pelvis, loc=<-1.0855 -0.2109 -8.7053>, rot=<0.0009 0.0355 0.0155 -0.9993>}
 calf_l {parent=thigh_l, loc=<-42.5284 -0.0019 -0.7297>, rot=<0.0006 -0.0161 -0.0391 -0.9991>}
 foot_l {parent=calf_l, loc=<-46.7401 -2.5014 -0.592>, rot=<0.0427 -0.0238 0.0353 -0.9982>}
 ball_l {parent=foot_l, loc=<-3.5658 10.8066 -0.1544>, rot=<0.0001 -0.0001 -0.7063 -0.7079>}
 calf_twist_02_l {parent=calf_l, loc=<-19.1546 -1.7943 -0.2638>, rot=<0.0237 0.0095 -0.0035 -0.9997>}
 calf_twist_01_l {parent=calf_l, loc=<-35.8483 -1.9197 -0.5794>, rot=<0.0237 0.0095 -0.0035 -0.9997>}
 deform_knee_in_base_l {parent=calf_l, loc=<0.8031 -5.0231 0.2456>, rot=<0 0 0 -1>}
 deform_knee_in_l {parent=deform_knee_in_base_l, loc=<0 0 0>, rot=<0 0 0 1>}
 deform_knee_out_base_l {parent=calf_l, loc=<0 4 0>, rot=<0 0 0 -1>}
 deform_knee_out_l {parent=deform_knee_out_base_l, loc=<0 0 0>, rot=<0 0 0 1>}
 thigh_twist_01_l {parent=thigh_l, loc=<-27.952 0 0>, rot=<0.0013 0 0 -1>}
 deform_groin_base_l {parent=thigh_l, loc=<0.4801 5.6776 -0.2486>, rot=<0 0 0 -1>}
 deform_groin_l {parent=deform_groin_base_l, loc=<0 0 0>, rot=<0 0 0 1>}
 thigh_r {parent=pelvis, loc=<-1.0855 -0.2109 8.7053>, rot=<0.0355 -0.0009 0.9993 0.0155>}
 calf_r {parent=thigh_r, loc=<42.5284 0.0019 0.7297>, rot=<0.0006 -0.0161 -0.0391 -0.9991>}
 foot_r {parent=calf_r, loc=<46.7401 2.5014 0.592>, rot=<0.0427 -0.0238 0.0353 -0.9982>}
 ball_r {parent=foot_r, loc=<3.5658 -10.8066 0.1544>, rot=<0.0001 -0.0001 -0.7063 -0.7079>}
 calf_twist_02_r {parent=calf_r, loc=<19.1546 1.7943 0.2638>, rot=<0.0237 0.0095 -0.0035 -0.9997>}
 calf_twist_01_r {parent=calf_r, loc=<35.8483 1.9197 0.5794>, rot=<0.0237 0.0095 -0.0035 -0.9997>}
 deform_knee_in_base_r {parent=calf_r, loc=<-0.8031 5.0231 -0.2457>, rot=<0 0 0 -1>}
 deform_knee_in_r {parent=deform_knee_in_base_r, loc=<0 0 0>, rot=<0 0 0 1>}
 deform_knee_out_base_r {parent=calf_r, loc=<0 -4 0>, rot=<0 0 0 -1>}
 deform_knee_out_r {parent=deform_knee_out_base_r, loc=<0 0 0>, rot=<0 0 0 1>}
 thigh_twist_01_r {parent=thigh_r, loc=<27.952 0 0>, rot=<0.0013 0 0 -1>}
 deform_groin_base_r {parent=thigh_r, loc=<-0.4801 -5.6776 0.2486>, rot=<0 0 0 -1>}
 deform_groin_r {parent=deform_groin_base_r, loc=<0 0 0>, rot=<0 0 0 1>}
 deform_glute_base_l {parent=pelvis, loc=<-7.3466 -9.0213 -9.1245>, rot=<0.0172 0.031 -0.4477 -0.8935>}
 deform_glute_l {parent=deform_glute_base_l, loc=<0 0 0>, rot=<0 0 0 1>}
 deform_glute_base_r {parent=pelvis, loc=<-7.3464 -9.0213 9.1245>, rot=<0.031 -0.0172 0.8935 -0.4477>}
 deform_glute_r {parent=deform_glute_base_r, loc=<0 0 0>, rot=<0 0 0 1>}
 base_bow {parent=pelvis, loc=<13.4516 -8.3049 0>, rot=<0.3251 -0.9457 0 0>}
 dyn_bow {parent=base_bow, loc=<0 0 0>, rot=<0 0 0 1>}
 base_main_tail {parent=dyn_bow, loc=<1.3732 0.5107 0>, rot=<0 0 0 1>}
 dyn_main_tail {parent=base_main_tail, loc=<0 0 0>, rot=<0 0 0 1>}
 base_tail_r_0 {parent=dyn_main_tail, loc=<5.3929 -0.5303 -1.2009>, rot=<-0.1052 -0.0821 -0.0364 -0.9904>}
 dyn_tail_r_0 {parent=base_tail_r_0, loc=<0 0 0>, rot=<0.0089 0.0068 0.0158 -0.9998>}
 dyn_tail_r_1 {parent=dyn_tail_r_0, loc=<9.8319 0.0001 0>, rot=<0.0035 -0.03 -0.1152 -0.9929>}
 dyn_tail_r_2 {parent=dyn_tail_r_1, loc=<9.6791 0 0>, rot=<0.0003 -0.014 -0.0199 -0.9997>}
 dyn_tail_r_3 {parent=dyn_tail_r_2, loc=<10.8537 0 0>, rot=<0.0008 -0.0187 -0.0453 -0.9988>}
 dyn_tail_r_4 {parent=dyn_tail_r_3, loc=<9.4802 0.0001 0>, rot=<0 0 0 1>}
 base_tail_l_0 {parent=dyn_main_tail, loc=<5.393 -0.5302 1.2009>, rot=<0.0953 0.0767 -0.0207 -0.9923>}
 dyn_tail_l_0 {parent=base_tail_l_0, loc=<0 0 0>, rot=<0 0 0 1>}
 dyn_tail_l_1 {parent=dyn_tail_l_0, loc=<9.8319 0 0>, rot=<-0.0035 0.03 -0.1152 -0.9929>}
 dyn_tail_l_2 {parent=dyn_tail_l_1, loc=<9.6791 0 0>, rot=<-0.0003 0.014 -0.0199 -0.9997>}
 dyn_tail_l_3 {parent=dyn_tail_l_2, loc=<10.8537 0 0>, rot=<-0.0008 0.0187 -0.0453 -0.9988>}
 dyn_tail_l_4 {parent=dyn_tail_l_3, loc=<9.4801 0 0>, rot=<0 0 0 1>}
 base_skirt_front_0 {parent=pelvis, loc=<5.6068 8.9531 0>, rot=<0 0 0.9935 0.1137>}
 dyn_skirt_front_0 {parent=base_skirt_front_0, loc=<0 0 0>, rot=<0 0 0 1>}
 dyn_skirt_front_1 {parent=dyn_skirt_front_0, loc=<11.1537 0 0>, rot=<0 0 -0.0175 -0.9998>}
 dyn_skirt_front_2 {parent=dyn_skirt_front_1, loc=<10.8946 0 0>, rot=<0 0 -0.0872 -0.9962>}
 base_skirt_front_l_0 {parent=pelvis, loc=<6.6776 6.7973 -7.6242>, rot=<-0.0996 0.2398 0.958 0.122>}
 dyn_skirt_front_l_0 {parent=base_skirt_front_l_0, loc=<0 0 0>, rot=<0 0 0 1>}
 dyn_skirt_front_l_1 {parent=dyn_skirt_front_l_0, loc=<10.9 0 0>, rot=<0 0 0 1>}
 dyn_skirt_front_l_2 {parent=dyn_skirt_front_l_1, loc=<10.9 0 0>, rot=<0 -0.0151 -0.0858 -0.9962>}
 base_skirt_side_l_0 {parent=pelvis, loc=<8.1649 -0.6016 -11.9415>, rot=<0.1163 -0.7015 -0.6933 -0.117>}
 dyn_skirt_side_l_0 {parent=base_skirt_side_l_0, loc=<0 0 0>, rot=<0 0 0 1>}
 dyn_skirt_side_l_1 {parent=dyn_skirt_side_l_0, loc=<10.9 0 0>, rot=<0 0 0 1>}
 dyn_skirt_side_l_2 {parent=dyn_skirt_side_l_1, loc=<10.9 0 0>, rot=<0 0 -0.0872 -0.9962>}
 base_skirt_back_side_l_0 {parent=pelvis, loc=<7.6548 -8.2897 -7.6636>, rot=<0.1451 -0.9446 -0.2835 -0.08>}
 dyn_skirt_back_side_l_0 {parent=base_skirt_back_side_l_0, loc=<0 0 0>, rot=<0 0 0 1>}
 dyn_skirt_back_side_l_1 {parent=dyn_skirt_back_side_l_0, loc=<10.9 0 0>, rot=<0 0 -0.0175 -0.9998>}
 dyn_skirt_back_side_l_2 {parent=dyn_skirt_back_side_l_1, loc=<10.9 0 0>, rot=<0 0 -0.0872 -0.9962>}
 base_skirt_back_0 {parent=pelvis, loc=<6.2879 -11.0007 0>, rot=<0.1127 -0.9936 0 0>}
 dyn_skirt_back_0 {parent=base_skirt_back_0, loc=<0 0 0>, rot=<0 0 0 1>}
 dyn_skirt_back_1 {parent=dyn_skirt_back_0, loc=<10.9 0 0>, rot=<0 0 -0.0698 -0.9976>}
 dyn_skirt_back_2 {parent=dyn_skirt_back_1, loc=<10.9 0 0>, rot=<0 0 -0.0262 -0.9997>}
 base_skirt_back_side_r_0 {parent=pelvis, loc=<7.6548 -8.2897 7.6636>, rot=<0.1451 -0.9446 0.2835 0.08>}
 dyn_skirt_back_side_r_0 {parent=base_skirt_back_side_r_0, loc=<0 0 0>, rot=<0 0 0 1>}
 dyn_skirt_back_side_r_1 {parent=dyn_skirt_back_side_r_0, loc=<10.9 0 0>, rot=<0 0 -0.0175 -0.9998>}
 dyn_skirt_back_side_r_2 {parent=dyn_skirt_back_side_r_1, loc=<10.9 0 0.0001>, rot=<0 0 -0.0872 -0.9962>}
 base_skirt_side_r_0 {parent=pelvis, loc=<8.1646 -0.6015 11.9415>, rot=<0.1163 -0.7015 0.6933 0.117>}
 dyn_skirt_side_r_0 {parent=base_skirt_side_r_0, loc=<0 0 0>, rot=<0 0 0 1>}
 dyn_skirt_side_r_1 {parent=dyn_skirt_side_r_0, loc=<10.8997 -0.0002 0>, rot=<0 0 0 1>}
 dyn_skirt_side_r_2 {parent=dyn_skirt_side_r_1, loc=<10.9 0 0>, rot=<0 0 -0.0872 -0.9962>}
 base_skirt_front_r_0 {parent=pelvis, loc=<6.6776 6.7973 7.6242>, rot=<0.0996 -0.2398 0.958 0.122>}
 dyn_skirt_front_r_0 {parent=base_skirt_front_r_0, loc=<0 0 0>, rot=<0 0 0 1>}
 dyn_skirt_front_r_1 {parent=dyn_skirt_front_r_0, loc=<10.9 0 0>, rot=<0 0 0 1>}
 dyn_skirt_front_r_2 {parent=dyn_skirt_front_r_1, loc=<10.9001 0 0>, rot=<0 0.0151 -0.0858 -0.9962>}
 for_aim_sh_bow_chest_l {parent=root, loc=<5.6318 11.0646 130.6002>, rot=<0.7542 0.27 -0.4954 -0.3361>}
 for_aim_sh_bow_chest_r {parent=root, loc=<-5.6318 11.0646 130.6005>, rot=<-0.4954 -0.3361 0.7542 0.27>}
 targeting_up {type=Curve, loc=<0 0 0>, rot=<0 0 0 1>}
]
variables [
 GetCrouchValue {type=double, default=0}
 isCrouching {type=bool, default=False}
 isTargeting {type=bool, default=False}
]
