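# Animation Blueprint for the Fortnite BrightLarva female body cosmetic, adding secondary motion to skirt, straps, bows and arm puffs.
# It copies the parent pose, layers in head-mesh facial bones, then runs a Control Rig and skirt bone offsets.
# RigidBody simulations, tail constraints and AnimDynamics follow, chained through cached poses.
# The EventGraph computes gravity, movement-based skirt alpha, zipline motion scaling and backbling flags, and fetches body and head meshes.
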AnimBlueprint /BRCosmetics/Characters/Player/Female/Medium/Bodies/F_MED_BrightLarva/Meshes/F_MED_BrightLarva_AnimBP
target_skeleton=&/F_MED_BrightLarva_Skeleton
graphs [
 AnimGraph {nodes [
   Output Pose(Result=Use cached pose 'DynamicsCache'()) {layerGroup=None}
   n1 = Layered blend per bone(
    BasePose=Copy Pose From Mesh() {bUseAttachedParent=true}
    BlendPoses_0=Copy Pose From Mesh(SourceMeshComponent="Get Face Mesh "() -> ("Face Mesh ":Skeletal Mesh Component Object Reference)) {bUseAttachedParent=true}
   ) {layerSetup [{branchFilters [
       {boneName=FACIAL_C_FacialRoot}
       {boneName=FACIAL_C_Jaw}
       {boneName=FACIAL_L_UnderChin}
       {boneName=FACIAL_R_UnderChin}
       {boneName=FACIAL_C_Jawline}
       {boneName=FACIAL_L_ChinSide}
       {boneName=FACIAL_R_ChinSide}
       {boneName=FACIAL_C_Chin}
       {boneName=FACIAL_L_UnderChin}
       {boneName=FACIAL_R_UnderChin}
       {boneName=FACIAL_R_Jawline}
       {boneName=FACIAL_L_Jawline}
      ]}]}
   Save cached pose 'Start_Cache'(Pose=@n1) {cacheName=Start_Cache}
   n2 = Control Rig!(Source=Use cached pose 'Start_Cache'(), Alpha, isCrouching:bool=Get bIsCrouching(), isTargeting:bool=Get bIsTargeting()) -> (Pose) {
    alphaBoolBlend {blendOption=Linear}
    controlRigAssetReference {blueprintRigClass=&/F_MED_BrightLarva_CtrlRig.F_MED_BrightLarva_CtrlRig_C}
   }
   Save cached pose 'controlRigCache'(Pose=@n2) {cacheName=controlRigCache}
   n3 = "Transform (Modify) Bone - Bone: base_skirt_back_0"!(
    ComponentPose=Local To Component(LocalPose=Use cached pose 'controlRigCache'())
    bAlphaBoolEnabled:bool=True
    Rotation="-0.000,-107.472,-0.000"
   ) -> (Pose) {
    alphaInputType=Bool
    rotationMode=BMM_Additive
    rotationSpace=BCS_BoneSpace
    translationSpace=BCS_BoneSpace
    alphaBoolBlend {blendInTime=0.2, blendOption=Linear, blendOutTime=0.2}
    boneToModify {boneName=base_skirt_back_0}
    rotation=<rot 0 -107.472 0>
   }
   n4 = "Transform (Modify) Bone - Bone: base_skirt_back_side_r_0"!(ComponentPose=@n3, bAlphaBoolEnabled:bool=True, Rotation="8.024,-17.200,-49.448") -> (Pose) {
    alphaInputType=Bool
    rotationMode=BMM_Additive
    rotationSpace=BCS_BoneSpace
    translationSpace=BCS_BoneSpace
    alphaBoolBlend {blendInTime=0.2, blendOption=Linear, blendOutTime=0.2}
    boneToModify {boneName=base_skirt_back_side_r_0}
    rotation=<rot 8.024 -17.2 -49.448>
   }
   n5 = "Transform (Modify) Bone - Bone: base_skirt_back_side_l_0"!(ComponentPose=@n4, bAlphaBoolEnabled:bool=True, Rotation="14.63,-56.833,46.022") -> (Pose) {
    alphaInputType=Bool
    rotationMode=BMM_Additive
    rotationSpace=BCS_BoneSpace
    translationSpace=BCS_BoneSpace
    alphaBoolBlend {blendInTime=0.2, blendOption=Linear, blendOutTime=0.2}
    boneToModify {boneName=base_skirt_back_side_l_0}
    rotation=<rot 14.63 -56.833 46.022>
   }
   n6 = "Transform (Modify) Bone - Bone: base_skirt_side_l_0"!(ComponentPose=@n5, bAlphaBoolEnabled:bool=True, Rotation="-37.625,-11.995,38.426") -> (Pose) {
    alphaInputType=Bool
    rotationMode=BMM_Additive
    rotationSpace=BCS_BoneSpace
    translationSpace=BCS_BoneSpace
    alphaBoolBlend {blendInTime=0.2, blendOption=Linear, blendOutTime=0.2}
    boneToModify {boneName=base_skirt_side_l_0}
    rotation=<rot -37.625 -11.995 38.426>
   }
   n7 = "Transform (Modify) Bone - Bone: base_skirt_front_l_0"!(ComponentPose=@n6, bAlphaBoolEnabled:bool=True, Rotation="15.164,-9.537,-0.518") -> (Pose) {
    alphaInputType=Bool
    rotationMode=BMM_Additive
    rotationSpace=BCS_BoneSpace
    translationSpace=BCS_BoneSpace
    alphaBoolBlend {blendInTime=0.2, blendOption=Linear, blendOutTime=0.2}
    boneToModify {boneName=base_skirt_front_l_0}
    rotation=<rot 15.164 -9.537 -0.518>
   }
   Save cached pose 'modifyTransformsCache'(Pose=Component To Local(ComponentPose=@n7)) {cacheName=modifyTransformsCache}
   n8 = RigidBody!(
    ComponentPose=Local To Component(LocalPose=Use cached pose 'modifyTransformsCache'())
    Alpha
    OverrideWorldGravity:Vector="0,0,0"
    ExternalForce
   ) -> (Pose) {
    bClampLinearTranslationLimitToRefPose=true
    bOverrideWorldGravity=true
    overridePhysicsAsset=&/F_MED_BrightLarva_Skirt_Physics
    simulationSpace=BaseBoneSpace
    alphaBoolBlend {blendOption=Linear}
    alphaScaleBiasClamp {bInterpResult=true, interpSpeedDecreasing=5, interpSpeedIncreasing=5}
    baseBoneRef {boneName=pelvis}
    simSpaceSettings {worldAlpha=0.3}
   }
   n9 = "Constraint - base_tail_l_0 (PP)"!(ComponentPose=@n8, Alpha=1, ConstraintWeights_0=0, ConstraintWeights_1:float=1) -> (Pose) {
    alphaBoolBlend {blendOption=Linear}
    boneToModify {boneName=base_tail_l_0}
    constraintSetup [{targetBone {boneName=dyn_skirt_back_side_l_0}}, {targetBone {boneName=dyn_skirt_back_0}}]
    constraintWeights [1,1]
   }
   n10 = "Constraint - base_tail_r_0 (PP)"!(ComponentPose=@n9, Alpha=1, ConstraintWeights_0=0, ConstraintWeights_1:float=1) -> (Pose) {
    alphaBoolBlend {blendOption=Linear}
    boneToModify {boneName=base_tail_r_0}
    constraintSetup [{targetBone {boneName=dyn_skirt_back_side_r_0}}, {targetBone {boneName=dyn_skirt_back_0}}]
    constraintWeights [1,1]
   }
   n11 = RigidBody(ComponentPose=@n10, OverrideWorldGravity="0,0,0") {
    bClampLinearTranslationLimitToRefPose=true
    bOverrideWorldGravity=true
    overridePhysicsAsset=&/F_MED_BrightLarva_Back_Straps_Physics
    simulationSpace=BaseBoneSpace
    alphaBoolBlend {blendOption=Linear}
    baseBoneRef {boneName=spine_02}
    simSpaceSettings {maxAngularAcceleration=6000, maxAngularVelocity=6000, maxLinearAcceleration=6000, maxLinearVelocity=6000, velocityScaleZ=0.9, worldAlpha=0.5}
   }
   n12 = RigidBody(ComponentPose=@n11, OverrideWorldGravity="0,0,0") {
    bClampLinearTranslationLimitToRefPose=true
    bOverrideWorldGravity=true
    overridePhysicsAsset=&/F_MED_BrightLarva_Chest_Bows_Physics
    simulationSpace=BaseBoneSpace
    alphaBoolBlend {blendOption=Linear}
    baseBoneRef {boneName=spine_05}
    simSpaceSettings {dampingAlpha=0.85, velocityScaleZ=0.85, worldAlpha=0.5}
   }
   n13 = Anim Dynamics - Bone: dyn_arm_puff_l!(ComponentPose=@n12, Alpha=1, ComponentLinearAccScale:Vector="0.5,0.5,0.5", ComponentLinearVelScale:Vector="0.5,0.5,0.5") -> (Pose) {
    angularDampingOverride=1
    angularSpringConstant=20000
    bAngularSpring=true
    bGravityOverrideInSimSpace=true
    bOverrideAngularDamping=true
    bUseGravityOverride=true
    gravityScale=0.35
    simulationSpace=RootRelative
    alphaBoolBlend {blendOption=Linear}
    boundBone {boneName=dyn_arm_puff_l}
    componentAppliedLinearAccClamp=<3000 3000 3000>
    componentLinearAccScale=<0.5 0.5 0.5>
    componentLinearVelScale=<0.5 0.5 0.5>
    gravityOverride {z=-980}
    physicsBodyDefinitions [{
      boundBone {boneName=dyn_arm_puff_l}
      boxExtents=<30 30 30>
      constraintSetup {angularLimitsMax=<0 10 10>, angularLimitsMin=<0 -10 -10>, angularTarget=<1 0 0>}
      localJointOffset=<8 0 0>
     }]
    simSpaceSettings {maxAngularAcceleration=35, maxAngularVelocity=7, simSpaceAngularAlpha=0.5}
   }
   n14 = Anim Dynamics - Bone: dyn_arm_puff_r!(ComponentPose=@n13, Alpha=1, ComponentLinearAccScale:Vector="0.5,0.5,0.5", ComponentLinearVelScale:Vector="0.5,0.5,0.5") -> (Pose) {
    angularDampingOverride=1
    angularSpringConstant=20000
    bAngularSpring=true
    bGravityOverrideInSimSpace=true
    bOverrideAngularDamping=true
    bUseGravityOverride=true
    gravityScale=0.35
    simulationSpace=RootRelative
    alphaBoolBlend {blendOption=Linear}
    boundBone {boneName=dyn_arm_puff_r}
    componentAppliedLinearAccClamp=<3000 3000 3000>
    componentLinearAccScale=<0.5 0.5 0.5>
    componentLinearVelScale=<0.5 0.5 0.5>
    gravityOverride {z=-980}
    physicsBodyDefinitions [{
      boundBone {boneName=dyn_arm_puff_r}
      boxExtents=<30 30 30>
      constraintSetup {angularLimitsMax=<0 10 10>, angularLimitsMin=<0 -10 -10>, angularTarget=<1 0 0>}
      localJointOffset=<8 0 0>
     }]
    simSpaceSettings {maxAngularAcceleration=35, maxAngularVelocity=7, simSpaceAngularAlpha=0.5}
   }
   n15 = "Transform (Modify) Bone - Bone: dyn_arm_puff_r"!(ComponentPose=@n14, Alpha=1, Scale="0.7,0.8,0.8") -> (Pose) {
    scaleMode=BMM_Additive
    scaleSpace=BCS_BoneSpace
    alphaBoolBlend {blendOption=Linear}
    alphaScaleBiasClamp {bInterpResult=true}
    boneToModify {boneName=dyn_arm_puff_r}
    scale=<0.7 0.8 0.8>
   }
   Save cached pose 'DynamicsCache'(Pose=Component To Local(ComponentPose=@n15)) {cacheName=DynamicsCache}
  ]}
 EventGraph {nodes [
   Event BlueprintUpdateAnimation()
   > n1 = "Gravity Override State Function (Parameter Asset) (Nativized)"(
    CharacterPartAnimInstance=Self-Reference()
    GameTimeInSeconds=GetGameTimeInSeconds()
    Parameters=&/F_MED_BrightLarva_Parameters
   )
   > Set GravityOverride(GravityOverride=@n1)
   > Set targetng up(targetng up:double=float - float(A=1, B=GetCurveValue(CurveName=targeting_up))) -> (Output_Get:double)
   n2 = Get bIsTacticalSprinting()
   n3 = Get bIsFalling()
   n4 = Get bIsJumping()
   > n5 = Branch(Condition=OR Boolean!(A=AND Boolean(A=@n4, B=@n2), B=AND Boolean(A=@n2, B=@n3), C:bool=Get bIsTacticalSprinting(), D:bool=Get bIsKneeSliding()) -> (ReturnValue))
   > Set alphaSkirtRigidBody float(alphaSkirtRigidBody float:double=0.6) -> (Output_Get:double)
   n6 = Branch(execute=@n5.else, Condition=OR Boolean!(A=Get bIsJumping(), B=Get bIsFalling()) -> (ReturnValue))
   > Set alphaSkirtRigidBody float(alphaSkirtRigidBody float=0.15)
   n7 = Get bIsCrouching()
   n8 = Branch(execute=@n6.else, Condition=AND Boolean(A=Get bIsSprinting(), B=NOT Boolean(A=@n7)))
   > Set alphaSkirtRigidBody float(alphaSkirtRigidBody float=0.6)
   Set alphaSkirtRigidBody float(execute=@n8.else, alphaSkirtRigidBody float=1)
   n9 = Event BlueprintProcessAnimInputEvent()
   n10 = IsNamed(self=@n9.AnimInputEvent, InEventName=aie.Player.Zipline.Entered)
   n11 = OR Boolean(A=Get bIsMoving2D(), B=Get bIsTacticalSprinting())
   n12 = Get bIsJumping()
   > n13 = Branch(Condition=OR Boolean(A=@n10, B=AND Boolean!(A=@n10, B=@n12, C:bool=@n11) -> (ReturnValue)))
   > Set animDyn on zipline(animDyn on zipline:Vector="0, 0, 0") -> (Output_Get:Vector)
   Branch(execute=@n13.else, Condition=IsNamed(self=@n9.AnimInputEvent, InEventName=aie.Player.Zipline.Exited))
   > Set animDyn on zipline(animDyn on zipline="0.5,0.5,0.5")
   Event OnCharacterCustomizationLoaded()
   > n14 = Sequence()
   > n15 = Cast To PlayerPawn_Athena(Object=TryGetPawnOwner())
   > "Set Face Mesh "("Face Mesh ":Skeletal Mesh Component Object Reference=GetSkeletalMeshForPartType(self=@n15, PartType=Head)) -> (Output_Get:Skeletal Mesh Component Object Reference)
   n16 = Cast To PlayerPawn_Athena(execute=@n14.then_1, Object=TryGetPawnOwner())
   > Set Body_Mesh(Body_Mesh=GetSkeletalMeshForPartType(self=@n16))
   n17 = CheckForBackbling()
   > Set isWearingBackbling(isWearingBackbling=AND Boolean(A=@n17, B=NOT Boolean(A=@n17.bIsBackblingACape), C=NOT Boolean(A=@n17.bIsBackblingLong)))
   > Set Is Wearing Cape(Is Wearing Cape=OR Boolean(A=@n17.bIsBackblingACape, B=@n17.bIsBackblingLong))
  ]}
]
variables [
 alphaSkirtRigidBody float {type=number}
 animDyn on zipline {type=Vector, default=<0.5 0.5 0.5>}
 Body_Mesh {type=SkeletalMeshComponent}
 "Face Mesh " {type=SkeletalMeshComponent}
 GravityOverride {type=Vector}
 Is Wearing Cape {type=boolean}
 isWearingBackbling {type=boolean}
 targetng up {type=number}
]
